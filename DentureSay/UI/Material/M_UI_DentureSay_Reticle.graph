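# Translucent UI material that draws a procedural, rotatable and scalable reticle, with a segmented ring, thicker corner tips, a triangle and a center marker.
# Circle, polygon and three-way radial segment distance fields are anti-aliased into named masks, then combined with max, multiply and subtract.
# A vertical step mask separates the top bracket from the other two for independent opacity control.
# Color blends from a two-color linear gradient toward white through the two-bracket mask.
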Material /DentureSay/UI/Material/M_UI_DentureSay_Reticle
graph {
 properties {bAutomaticallySetUsageInEditor=false, blendMode=BLEND_Translucent, materialDomain=MD_UI, usedShadingModels=Default Lit}
 nodes [
  n1 = NamedRerouteUsage()
  n2 = Constant() {r=0.2}
  n3 = Constant() {r=0.1}
  n4 = Constant() {r=0.25}
  n5 = Max(A=NamedRerouteUsage(), B=Multiply(A=NamedRerouteUsage(), B=NamedRerouteUsage()))
  n6 = NamedRerouteUsage()
  n7 = Max(A=Saturate(None=Subtract(A=@n5, B=@n6)), B=Multiply(A=NamedRerouteUsage(), B=NamedRerouteUsage()))
  n8 = NamedRerouteUsage()
  n9 = Multiply(A=Max(A=@n7, B=NamedRerouteUsage()), B=@n8)
  LinearInterpolate(
   A=LinearInterpolate(
    A=VectorParameter() {parameterName=Fill Color 2, defaultValue=<rgba 0 0.8768 1 1>}.RGB
    B=VectorParameter() {parameterName=Fill Color 1, defaultValue=<rgba 1 0.2755 0 1>}.RGB
    Alpha=MaterialFunctionCall(UVs=@n1, GradientStart=@n2, GradientEnd=@n3, RotationAngle=@n4) {materialFunction=&MF_UI_LinearGradient}.Result
   )
   Alpha=@n9
  ) {constA=1}
  n10 = OneMinus(None=@n8)
  Max(
   A=Multiply(A=@n9, B=ScalarParameter() {defaultValue=1, parameterName=TwoBracketsOpacity})
   B=Multiply(A=Multiply(A=@n7, B=@n10), B=ScalarParameter() {defaultValue=1, parameterName=TopBracketOpacity})
  )
  n11 = MaterialFunctionCall() {materialFunction=&GetUserInterfaceUV}
  n12 = ScalarParameter() {defaultValue=0.5, parameterName=ReticleRotation}
  NamedRerouteDeclaration(None=MaterialFunctionCall(
    UVs=MaterialFunctionCall(UVs=@n11.Normalized UV, "Rotation Angle (0-1)"=@n12) {materialFunction=&CustomRotator}.Rotated Values
    Texture Scale=ScalarParameter() {defaultValue=1, parameterName=ReticleSize}
   ) {materialFunction=&ScaleUVsByCenter}.UVs) {name=ReticleUV, nodeColor=<rgba 0 0 0 1>}
  n13 = NamedRerouteUsage()
  n14 = Constant() {r=0.7}
  n15 = Constant() {r=1}
  n16 = MaterialFunctionCall(UVs=@n13, Size=@n14, OutlineThickness=ScalarParameter() {defaultValue=0.03, parameterName=RingThickness}, OutlineAlignment=@n15) {materialFunction=&MF_UI_Circle}
  NamedRerouteDeclaration(None=MaterialFunctionCall(DistanceField=@n16.SDFOutline) {materialFunction=&MF_UI_AntiAliasing}.Inner) {name=Ring, nodeColor=<rgba 0 0 0 1>}
  n17 = NamedRerouteUsage()
  n18 = MaterialFunctionCall(UVs=@n17, Size=Constant() {r=0.05}) {materialFunction=&MF_UI_Circle}
  n19 = Constant() {r=0.5}
  n20 = MaterialFunctionCall!(
   UVs=Add(
    A=MaterialFunctionCall(UVs=@n17, "Rotation Angle (0-1)"=@n19) {materialFunction=&CustomRotator}.Rotated Values
    B=Constant2Vector() {g=-0.04}
   )
   Size=Constant() {r=0.07}
   Sides=Constant() {r=3}
   InnerAngle
   OutlineThickness
   OutlineAlignment
   GlowMin
   GlowMax
  ) {materialFunction=/FortUILibrary/MaterialFunctions/Shapes/MF_UI_RegularPolygon}
  NamedRerouteDeclaration(None=Max(
    A=MaterialFunctionCall(DistanceField=@n18.SDF) {materialFunction=&MF_UI_AntiAliasing}.Inner
    B=MaterialFunctionCall(DistanceField=@n20.SDF) {materialFunction=&MF_UI_AntiAliasing}.Inner
   )) {name=CenterDot, nodeColor=<rgba 0 0 0 1>}
  n21 = MaterialFunctionCall!(
   UV=NamedRerouteUsage()
   Segment Count=Constant() {r=3}
   Segment Spacing=ScalarParameter() {defaultValue=0.3, parameterName=SegmentSpacing}
   Fixed Spacing
  ) {materialFunction=/FortUILibrary/MaterialFunctions/Basics/MF_UI_RadialSegments}
  NamedRerouteDeclaration(None=MaterialFunctionCall(DistanceField=@n21.Segmented Triangles Gradient, Threshold=ScalarParameter() {defaultValue=0.4, parameterName=CornersThickness}) {materialFunction=&MF_UI_AntiAliasing}.Outer) {name=Corners, nodeColor=<rgba 0 0 0 1>}
  NamedRerouteDeclaration(None=MaterialFunctionCall(DistanceField=@n21.Segmented Triangles Gradient, Threshold=Constant() {r=0.5}) {materialFunction=&MF_UI_AntiAliasing}.Outer) {name=Cut, nodeColor=<rgba 0 0 0 1>}
  n22 = ScalarParameter() {defaultValue=0.07, parameterName=TipsHeight}
  NamedRerouteDeclaration(None=MaterialFunctionCall(DistanceField=MaterialFunctionCall(UVs=@n13, Size=@n14, OutlineThickness=@n22, OutlineAlignment=@n15) {materialFunction=&MF_UI_Circle}.SDFOutline) {materialFunction=&MF_UI_AntiAliasing}.Inner) {name=CornerMask, nodeColor=<rgba 0 0 0 1>}
  n23 = NamedRerouteUsage()
  n24 = AppendVector(A=Constant() {r=0.5}, B=ScalarParameter() {defaultValue=1, parameterName=Triangle position})
  n25 = MaterialFunctionCall(UVs=Add(A=Multiply(A=@n23, B=@n24), B=Constant2Vector() {g=-0.39, r=0.25}), "Rotation Angle (0-1)"=Constant() {r=0.5}) {materialFunction=&CustomRotator}
  n26 = Constant() {r=0.125}
  n27 = Constant() {r=3}
  NamedRerouteDeclaration(None=MaterialFunctionCall(DistanceField=MaterialFunctionCall(UVs=@n25.Rotated Values, Size=@n26, Sides=@n27) {materialFunction=&MF_UI_RegularPolygon}.SDF) {materialFunction=&MF_UI_AntiAliasing}.Inner) {name=Triangle, nodeColor=<rgba 0 0 0 1>}
  NamedRerouteDeclaration(None=MaterialFunctionCall(DistanceField=@n16.SDF) {materialFunction=&MF_UI_AntiAliasing}.Outer) {name=TriangleMask, nodeColor=<rgba 0 0 0 1>}
  n28 = NamedRerouteUsage()
  NamedRerouteDeclaration(None=Step(Y=ComponentMask(None=@n28) {g=true}) {constX=0.7}) {name=BracketOpacity, nodeColor=<rgba 0 0 0 1>}
 ]
}
