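# Translucent UI material for DenturySay controls that draws a rounded shape with a gradient fill and gradient outline.
# A rounded-mask function with fixed per-corner radii and a thickness parameter supplies Stroke and Fill masks through named reroutes.
# Two rotated linear gradients blend parameterized color pairs, each masked by a lerp from black, one using a subtraction of masks.
# A Max merges both layers into the Color reroute.
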Material /DentureSay/UI/Material/M_UI_DenturySay_Controls
graph {
 properties {bAutomaticallySetUsageInEditor=false, blendMode=BLEND_Translucent, materialDomain=MD_UI, usedShadingModels=Default Lit}
 nodes [
  n1 = MaterialFunctionCall(
   UV=TextureCoordinate()
   Corner Radius=Constant4Vector() {constant=<rgba 0 0 1 1>}.RGBA
   Outline Thickness=ScalarParameter() {defaultValue=30, parameterName=Outline Thickness}
  ) {materialFunction=&MF_UI_CoreUI_RoundedMask}
  NamedRerouteDeclaration(None=@n1.Outline Mask) {name=Stroke, nodeColor=<rgba 0 0 0 1>}
  NamedRerouteDeclaration(None=@n1.Shape Mask) {name=Fill, nodeColor=<rgba 0 0 0 1>}
  n2 = TextureCoordinate()
  n3 = Constant() {r=0.75}
  n4 = LinearInterpolate(
   B=LinearInterpolate(
    A=VectorParameter() {parameterName=ColorStroke1, defaultValue=<rgba 0.0802 0.2051 0.7835 1>}.RGBA
    B=VectorParameter() {parameterName=ColorStroke2, defaultValue=<rgba 0.0802 0.2051 0.7835 1>}.RGBA
    Alpha=MaterialFunctionCall(UVs=@n2, RotationAngle=@n3) {materialFunction=&MF_UI_LinearGradient}.Result
   )
   Alpha=NamedRerouteUsage()
  )
  n5 = TextureCoordinate()
  n6 = Constant() {r=0.75}
  n7 = LinearInterpolate(
   B=LinearInterpolate(
    A=VectorParameter() {parameterName=ColorFill1, defaultValue=<rgba 0 0 0 1>}.RGBA
    B=VectorParameter() {parameterName=ColorFill2, defaultValue=<rgba 0.0012 0.0382 0.4969 1>}.RGBA
    Alpha=MaterialFunctionCall(UVs=@n5, RotationAngle=@n6) {materialFunction=&MF_UI_LinearGradient}.Result
   )
   Alpha=Subtract(A=NamedRerouteUsage(), B=NamedRerouteUsage())
  )
  NamedRerouteDeclaration(None=Max(
    A=NamedRerouteDeclaration(None=@n4) {name=StrokeColor, nodeColor=<rgba 0 0 0 1>}
    B=NamedRerouteDeclaration(None=@n7) {name=FillInside, nodeColor=<rgba 0 0 0 1>}
   )) {name=Color, nodeColor=<rgba 0 0 0 1>}
  NamedRerouteUsage()
  NamedRerouteUsage()
  Saturate(None=@n1.Outline Mask)
 ]
}
