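# Translucent UI material that draws animated yellow lightning for the DentureSay interface.
# A radial gradient blends two yellow tints for the color.
# A lightning texture, scaled and rotated, jumps between UV offsets driven by quantized time for a stepped flicker.
# Its smoothstepped red channel, minus a soft circular SDF mask, is multiplied by an opacity parameter, with size, speed and mask size also parameterized.
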Material /DentureSay/UI/Material/M_UI_DenturySay_Lightning
graph {
 properties {bAutomaticallySetUsageInEditor=false, blendMode=BLEND_Translucent, materialDomain=MD_UI, usedShadingModels=Default Lit}
 nodes [
  LinearInterpolate(
   A=Constant4Vector() {constant=<rgba 1 0.959 0 1>}.RGBA
   B=Constant4Vector() {constant=<rgba 1 0.974 0.3646 1>}.RGBA
   Alpha=MaterialFunctionCall(UV=TextureCoordinate()) {materialFunction=&MF_UI_RadialGradient}.Smooth
  )
  n1 = MaterialFunctionCall(
   UVs=MaterialFunctionCall(UV=TextureCoordinate(), TextureSize=Constant() {r=5}) {materialFunction=&MF_UI_UVScaling}.ScaleY
   Texture Scale=ScalarParameter() {defaultValue=3, parameterName=Texture Size}
  ) {materialFunction=&ScaleUVsByCenter}
  n2 = Constant() {r=0.3}
  n3 = MaterialFunctionCall!("Param (0-1)"=Frac(None=Time()), Tiles=ScalarParameter() {defaultValue=10, parameterName=LightningSpeed}, Apply Offset) {materialFunction=/Game/Environments/Papaya/Materials/MaterialFunctions/MF_Papaya_Quantize_Scalar}
  n4 = Multiply(A=@n3.Result) {constB=3}
  n5 = TextureSample(UVs=Add(
    A=MaterialFunctionCall(UVs=@n1.UVs, "Rotation Angle (0-1)"=@n2) {materialFunction=&CustomRotator}.Rotated Values
    B=AppendVector(A=@n3.Result, B=@n4)
   )) {texture=/DentureSay/UI/Textures/T_UI_DentureSay_Lightning}
  n6 = MaterialFunctionCall(
   UVs=MaterialFunctionCall() {materialFunction=&GetUserInterfaceUV}.9-Slice UV
   Size=ScalarParameter() {defaultValue=0.5, parameterName=Mask Size}
  ) {materialFunction=&MF_UI_Circle}
  n7 = Constant() {r=0.5}
  Multiply(
   A=Subtract(
    A=SmoothStep(Value=@n5.R) {constMax=0.6}
    B=MaterialFunctionCall(DistanceField=@n6.SDF, Smoothing=@n7) {materialFunction=&MF_UI_AntiAliasing}.Outer
   )
   B=ScalarParameter() {defaultValue=1, parameterName=Lightning Opacity}
  )
 ]
}
