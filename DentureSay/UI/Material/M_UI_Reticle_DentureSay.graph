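# Translucent UI material that draws a reticle procedurally from SDF box, circle and polygon functions.
# Named reroutes hold the center dot, six rotated tick marks, a direction icon and a segmented outer ring with separate opacity parameters for top and outer brackets.
# A box mask separates a two-color gradient fill from areas tinted by Reticle Color.
# A Reticle Rotation parameter rotates the UVs published as TextureCoord.
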Material /DentureSay/UI/Material/M_UI_Reticle_DentureSay
graph {
 properties {bAutomaticallySetUsageInEditor=false, blendMode=BLEND_Translucent, materialDomain=MD_UI, usedShadingModels=Default Lit}
 nodes [
  Constant()
  Constant()
  Constant()
  Constant()
  Constant()
  Constant()
  n1 = VectorParameter() {parameterName=Fill Color 1, defaultValue=<rgba 1 0.2755 0 1>}
  n2 = VectorParameter() {parameterName=Fill Color 2, defaultValue=<rgba 0 0.8768 1 1>}
  n3 = MaterialFunctionCall(UVs=TextureCoordinate(), GradientStart=Constant() {r=1}, GradientEnd=Constant() {r=0.8}, RotationAngle=Constant() {r=0.25}) {materialFunction=&MF_UI_LinearGradient}
  n4 = NamedRerouteUsage()
  n5 = Constant2Vector() {g=0.4}
  n6 = MaterialFunctionCall(DistanceField=MaterialFunctionCall(UVs=Add(A=@n4, B=@n5), Size=Constant2Vector() {g=0.5, r=0.6}) {materialFunction=&MF_UI_Box}.SDF) {materialFunction=&MF_UI_AntiAliasing}
  n7 = NamedRerouteUsage()
  n8 = Min(A=@n6.Inner, B=@n7)
  n9 = Subtract(A=@n7, B=@n6.Inner)
  n10 = NamedRerouteUsage()
  Max(
   A=Multiply(A=LinearInterpolate(A=@n1.RGB, B=@n2.RGB, Alpha=@n3.Result), B=@n8)
   B=Multiply(A=VectorParameter() {parameterName=Reticle Color, defaultValue=<rgba 1 1 1 1>}.RGB, B=Max(A=@n9, B=@n10))
  )
  NamedRerouteDeclaration(None=MaterialFunctionCall(UVs=TextureCoordinate(), "Rotation Angle (0-1)"=ScalarParameter() {defaultValue=0.165, parameterName=Reticle Rotation}) {materialFunction=&CustomRotator}.Rotated Values) {name=TextureCoord, nodeColor=<rgba 0 0 0 1>}
  NamedRerouteDeclaration(None=@n6.Inner) {name=SegmentMask, nodeColor=<rgba 0 0 0 1>}
  n11 = NamedRerouteUsage()
  n12 = Constant2Vector() {g=0.4}
  n13 = Constant2Vector() {g=0.5, r=0.5}
  n14 = Constant() {r=0.404}
  n15 = MaterialFunctionCall(
   UVs=MaterialFunctionCall(UVs=Add(A=@n11, B=@n12) {constB=0}, Rotation Center=Add(A=@n13, B=@n12), "Rotation Angle (0-1)"=OneMinus(None=@n14)) {materialFunction=&CustomRotator}.Rotated Values
   Size=NamedRerouteUsage()
  ) {materialFunction=&MF_UI_Box}
  n16 = Constant() {r=0.01}
  n17 = NamedRerouteUsage()
  n18 = Constant2Vector() {g=0.4}
  n19 = Constant2Vector() {g=0.5, r=0.5}
  n20 = Constant() {r=0.596}
  n21 = MaterialFunctionCall(
   UVs=MaterialFunctionCall(UVs=Add(A=@n17, B=@n18) {constB=0}, Rotation Center=Add(A=@n19, B=@n18), "Rotation Angle (0-1)"=OneMinus(None=@n20)) {materialFunction=&CustomRotator}.Rotated Values
   Size=NamedRerouteUsage()
  ) {materialFunction=&MF_UI_Box}
  n22 = Constant() {r=0.01}
  NamedRerouteDeclaration(None=Add(
    A=MaterialFunctionCall(DistanceField=@n15.SDF, Smoothing=@n16) {materialFunction=&MF_UI_AntiAliasing}.Inner
    B=MaterialFunctionCall(DistanceField=@n21.SDF, Smoothing=@n22) {materialFunction=&MF_UI_AntiAliasing}.Inner
   )) {name=Tick2, nodeColor=<rgba 0 0 0 1>}
  NamedRerouteDeclaration(None=TextureCoordinate()) {name=TC, nodeColor=<rgba 0 0 0 1>}
  n23 = MaterialFunctionCall(UVs=NamedRerouteUsage(), Size=Constant() {r=0.05}) {materialFunction=&MF_UI_Circle}
  n24 = MaterialFunctionCall!(
   UVs=Add(A=NamedRerouteUsage(), B=Constant2Vector() {g=-0.04})
   Size=Constant() {r=0.07}
   Sides=Constant() {r=3}
   InnerAngle
   OutlineThickness
   OutlineAlignment
   GlowMin
   GlowMax
  ) {materialFunction=/FortUILibrary/MaterialFunctions/Shapes/MF_UI_RegularPolygon}
  NamedRerouteDeclaration(None=Max(
    A=MaterialFunctionCall(DistanceField=@n23.SDF) {materialFunction=&MF_UI_AntiAliasing}.Inner
    B=MaterialFunctionCall(DistanceField=@n24.SDF) {materialFunction=&MF_UI_AntiAliasing}.Inner
   )) {name=CenterDot, nodeColor=<rgba 0 0 0 1>}
  n25 = NamedRerouteUsage()
  n26 = Constant2Vector() {g=0.4}
  n27 = Constant2Vector() {g=0.5, r=0.5}
  n28 = Constant() {r=0.736}
  n29 = MaterialFunctionCall(
   UVs=MaterialFunctionCall(UVs=Add(A=@n25, B=@n26) {constB=0}, Rotation Center=Add(A=@n27, B=@n26), "Rotation Angle (0-1)"=OneMinus(None=@n28)) {materialFunction=&CustomRotator}.Rotated Values
   Size=NamedRerouteUsage()
  ) {materialFunction=&MF_UI_Box}
  n30 = Constant() {r=0.01}
  n31 = NamedRerouteUsage()
  n32 = Constant2Vector() {g=0.4}
  n33 = Constant2Vector() {g=0.5, r=0.5}
  n34 = Constant() {r=0.93}
  n35 = MaterialFunctionCall(
   UVs=MaterialFunctionCall(UVs=Add(A=@n31, B=@n32) {constB=0}, Rotation Center=Add(A=@n33, B=@n32), "Rotation Angle (0-1)"=OneMinus(None=@n34)) {materialFunction=&CustomRotator}.Rotated Values
   Size=NamedRerouteUsage()
  ) {materialFunction=&MF_UI_Box}
  n36 = Constant() {r=0.01}
  NamedRerouteDeclaration(None=Add(
    A=MaterialFunctionCall(DistanceField=@n29.SDF, Smoothing=@n30) {materialFunction=&MF_UI_AntiAliasing}.Inner
    B=MaterialFunctionCall(DistanceField=@n35.SDF, Smoothing=@n36) {materialFunction=&MF_UI_AntiAliasing}.Inner
   )) {name=Tick3, nodeColor=<rgba 0 0 0 1>}
  n37 = NamedRerouteUsage()
  n38 = NamedRerouteUsage()
  NamedRerouteDeclaration(None=Add(A=Add(A=@n37, B=@n38), B=NamedRerouteUsage())) {name=Ticks, nodeColor=<rgba 0 0 0 1>}
  NamedRerouteDeclaration(None=Max(A=NamedRerouteUsage(), B=NamedRerouteUsage())) {name=FullReticle, nodeColor=<rgba 0 0 0 1>}
  NamedRerouteDeclaration(None=@n8) {name=UpperBracket, nodeColor=<rgba 0 0 0 1>}
  NamedRerouteDeclaration(None=Multiply(A=NamedRerouteUsage(), B=Constant() {r=1})) {name=UB, nodeColor=<rgba 0 0 0 1>}
  NamedRerouteDeclaration(None=@n6.Inner) {name=Block, nodeColor=<rgba 0 0 0 1>}
  NamedRerouteDeclaration(None=Constant2Vector() {g=0.04, r=0.02}) {name=TickSizes, nodeColor=<rgba 0 0 0 1>}
  n39 = MaterialFunctionCall(UVs=NamedRerouteUsage(), Size=Constant() {r=0.75}, OutlineThickness=Constant() {r=0.03}) {materialFunction=&MF_UI_Circle}
  NamedRerouteDeclaration(None=MaterialFunctionCall(DistanceField=@n39.SDFOutline) {materialFunction=&MF_UI_AntiAliasing}.Inner) {name=Outer Bracket, nodeColor=<rgba 0 0 0 1>}
  n40 = TextureCoordinate()
  n41 = Constant2Vector() {g=0.28}
  n42 = MaterialFunctionCall(
   UVs=MaterialFunctionCall(UVs=Add(A=@n40, B=@n41), "Rotation Angle (0-1)"=Constant() {r=0.125}) {materialFunction=&CustomRotator}.Rotated Values
   Size=Constant() {r=0.25}
  ) {materialFunction=&MF_UI_Box}
  NamedRerouteDeclaration(None=Min(
    A=MaterialFunctionCall(DistanceField=@n42.SDF) {materialFunction=&MF_UI_AntiAliasing}.Inner
    B=MaterialFunctionCall(DistanceField=@n39.SDF) {materialFunction=&MF_UI_AntiAliasing}.Outer
   )) {name=DirectionIcon, nodeColor=<rgba 0 0 0 1>}
  NamedRerouteDeclaration(None=MaterialFunctionCall!(UV=NamedRerouteUsage(), Segment Count=Constant() {r=3}, Segment Spacing=Constant() {r=0.6}, Fixed Spacing) {materialFunction=/FortUILibrary/MaterialFunctions/Basics/MF_UI_RadialSegments}.Segmented Triangles Fill) {name=Segment Outer Circle, nodeColor=<rgba 0 0 0 1>}
  n43 = Max(A=Subtract(A=NamedRerouteUsage(), B=NamedRerouteUsage()), B=NamedRerouteUsage())
  n44 = NamedRerouteUsage()
  n45 = Subtract(A=@n43, B=@n44)
  n46 = ScalarParameter() {defaultValue=1, parameterName=Outer Brackets Opacity}
  n47 = Max(A=Subtract(A=@n43, B=OneMinus(None=@n44)), B=NamedRerouteUsage())
  n48 = ScalarParameter() {defaultValue=1, parameterName=Top Bracket Opacity}
  NamedRerouteDeclaration(None=Max(A=Multiply(A=@n45, B=@n46), B=Multiply(A=@n47, B=@n48))) {name=Outer Bracket Completed, nodeColor=<rgba 0 0 0 1>}
  n49 = NamedRerouteUsage()
  n50 = Constant2Vector() {g=0.4}
  n51 = Constant2Vector() {g=0.5, r=0.5}
  n52 = Constant() {r=0.07}
  n53 = MaterialFunctionCall(
   UVs=MaterialFunctionCall(UVs=Add(A=@n49, B=@n50) {constB=0}, Rotation Center=Add(A=@n51, B=@n50), "Rotation Angle (0-1)"=OneMinus(None=@n52)) {materialFunction=&CustomRotator}.Rotated Values
   Size=NamedRerouteUsage()
  ) {materialFunction=&MF_UI_Box}
  n54 = Constant() {r=0.01}
  n55 = NamedRerouteUsage()
  n56 = Constant2Vector() {g=0.4}
  n57 = Constant2Vector() {g=0.5, r=0.5}
  n58 = Constant() {r=0.264}
  n59 = MaterialFunctionCall(
   UVs=MaterialFunctionCall(UVs=Add(A=@n55, B=@n56) {constB=0}, Rotation Center=Add(A=@n57, B=@n56), "Rotation Angle (0-1)"=OneMinus(None=@n58)) {materialFunction=&CustomRotator}.Rotated Values
   Size=NamedRerouteUsage()
  ) {materialFunction=&MF_UI_Box}
  n60 = Constant() {r=0.01}
  NamedRerouteDeclaration(None=Add(
    A=MaterialFunctionCall(DistanceField=@n53.SDF, Smoothing=@n54) {materialFunction=&MF_UI_AntiAliasing}.Inner
    B=MaterialFunctionCall(DistanceField=@n59.SDF, Smoothing=@n60) {materialFunction=&MF_UI_AntiAliasing}.Inner
   )) {name=Tick1, nodeColor=<rgba 0 0 0 1>}
  NamedRerouteUsage()
 ]
}
