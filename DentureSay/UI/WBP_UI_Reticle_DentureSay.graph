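# Charged-weapon reticle extension widget for DentureSay, showing primary combo hits and secondary ability cooldown.
# On view target changes, it binds the pawn's Mantis technique-start event and async listeners for the cooldown gameplay tag.
# Technique index selects progressively faster bracket scale-punch animations and starts a timer.
# Tag add or removal plays lightning and bracket animations and toggles a looping timer that drives cooldown progress on the reticle's dynamic material.
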WidgetBlueprint /DentureSay/UI/WBP_UI_Reticle_DentureSay
animations [
 anim_Idle {
  binding_count=2
  length_seconds=0.5
  object=&self:anim_Idle
  sequence {bindings [
    Image_Lightning {
     id=#1
     possessed_class=Image
     tracks [{
       type=MovieSceneWidgetMaterialTrack
       sections [{
         type=MovieSceneParameterSection
         end_frame=10
         start_frame=0
         channels [
          Mask Size {type=MovieSceneScriptingFloatChannel, const [0.3,1]}
          Lightning Opacity {type=MovieSceneScriptingFloatChannel, const [1,1]}
         ]
        }]
      }]
    }
    Reticle_Material {
     id=#2
     possessed_class=Image
     tracks [{
       type=MovieSceneWidgetMaterialTrack
       sections [{
         type=MovieSceneParameterSection
         end_frame=10
         start_frame=0
         channels [Outer Brackets Opacity {type=MovieSceneScriptingFloatChannel, const [1,1]}]
        }]
      }]
    }
   ]}
 }
 anim_1HitCombo {
  binding_count=1
  length_seconds=0.15
  object=&self:anim_1HitCombo
  sequence {bindings [Reticle_Material {
     id=#3
     possessed_class=Image
     tracks [
      {
       type=MovieSceneWidgetMaterialTrack
       sections [{
         type=MovieSceneParameterSection
         end_frame=3
         start_frame=0
         channels [Outer Brackets Opacity {type=MovieSceneScriptingFloatChannel, const [1,1]}]
        }]
      }
      {
       type=MovieScene2DTransformTrack
       sections [{
         type=MovieScene2DTransformSection
         end_frame=3
         start_frame=0
         channels [
          Translation.X {type=MovieSceneScriptingFloatChannel, const [0,1]}
          Translation.Y {type=MovieSceneScriptingFloatChannel, const [0,1]}
          Angle {type=MovieSceneScriptingFloatChannel, const [0,1]}
          Scale.X {type=MovieSceneScriptingFloatChannel, keys [[0,1],[1,0.9],[3,1]]}
          Scale.Y {type=MovieSceneScriptingFloatChannel, keys [[0,1],[1,0.9],[3,1]]}
          Shear.X {type=MovieSceneScriptingFloatChannel, const [0,1]}
          Shear.Y {type=MovieSceneScriptingFloatChannel, const [0,1]}
         ]
        }]
      }
     ]
    }]}
 }
 anim_2HitCombo {
  binding_count=1
  length_seconds=0.1
  object=&self:anim_2HitCombo
  sequence {bindings [Reticle_Material {
     id=#3
     possessed_class=Image
     tracks [
      {
       type=MovieSceneWidgetMaterialTrack
       sections [{
         type=MovieSceneParameterSection
         end_frame=2
         start_frame=0
         channels [Outer Brackets Opacity {type=MovieSceneScriptingFloatChannel, const [1,1]}]
        }]
      }
      {
       type=MovieScene2DTransformTrack
       sections [{
         type=MovieScene2DTransformSection
         end_frame=2
         start_frame=0
         channels [
          Translation.X {type=MovieSceneScriptingFloatChannel, const [0,1]}
          Translation.Y {type=MovieSceneScriptingFloatChannel, const [0,1]}
          Angle {type=MovieSceneScriptingFloatChannel, const [0,1]}
          Scale.X {type=MovieSceneScriptingFloatChannel, keys [[0,1],[0.6667,0.9],[2,1]]}
          Scale.Y {type=MovieSceneScriptingFloatChannel, keys [[0,1],[0.6667,0.9],[2,1]]}
          Shear.X {type=MovieSceneScriptingFloatChannel, const [0,1]}
          Shear.Y {type=MovieSceneScriptingFloatChannel, const [0,1]}
         ]
        }]
      }
     ]
    }]}
 }
 anim_3HitCombo {
  binding_count=1
  length_seconds=0.05
  object=&self:anim_3HitCombo
  sequence {bindings [Reticle_Material {
     id=#3
     possessed_class=Image
     tracks [
      {
       type=MovieSceneWidgetMaterialTrack
       sections [{
         type=MovieSceneParameterSection
         end_frame=1
         start_frame=0
         channels [Outer Brackets Opacity {type=MovieSceneScriptingFloatChannel, const [1,1]}]
        }]
      }
      {
       type=MovieScene2DTransformTrack
       sections [{
         type=MovieScene2DTransformSection
         end_frame=1
         start_frame=0
         channels [
          Translation.X {type=MovieSceneScriptingFloatChannel, const [0,1]}
          Translation.Y {type=MovieSceneScriptingFloatChannel, const [0,1]}
          Angle {type=MovieSceneScriptingFloatChannel, const [0,1]}
          Scale.X {type=MovieSceneScriptingFloatChannel, keys [[0,1],[0.3333,0.9],[1,1]]}
          Scale.Y {type=MovieSceneScriptingFloatChannel, keys [[0,1],[0.3333,0.9],[1,1]]}
          Shear.X {type=MovieSceneScriptingFloatChannel, const [0,1]}
          Shear.Y {type=MovieSceneScriptingFloatChannel, const [0,1]}
         ]
        }]
      }
     ]
    }]}
 }
 anim_SecondaryAttack {
  binding_count=2
  length_seconds=1.25
  object=&self:anim_SecondaryAttack
  sequence {bindings [
    Image_Lightning {
     id=#4
     possessed_class=Image
     tracks [{
       type=MovieSceneWidgetMaterialTrack
       sections [{
         type=MovieSceneParameterSection
         end_frame=25
         start_frame=0
         channels [
          Mask Size {type=MovieSceneScriptingFloatChannel, keys [[0,0.3],[4.6363,0.5],[13,0.5],[15,0.3]]}
          Lightning Opacity {type=MovieSceneScriptingFloatChannel, thinned_from=4, keys [[0,1],[13,1],[16,0]]}
         ]
        }]
      }]
    }
    Reticle_Material {
     id=#2
     possessed_class=Image
     tracks [{
       type=MovieSceneWidgetMaterialTrack
       sections [{
         type=MovieSceneParameterSection
         end_frame=25
         start_frame=0
         channels [
          Outer Brackets Opacity {type=MovieSceneScriptingFloatChannel, const [1,1]}
          TopBracketOpacity {type=MovieSceneScriptingFloatChannel, keys [[0,1],[25,0.5]]}
         ]
        }]
      }]
    }
   ]}
 }
 anim_SecondaryAttackCooldownFinished {
  binding_count=2
  length_seconds=0.15
  object=&self:anim_SecondaryAttackCooldownFinished
  sequence {bindings [
    Image_Lightning {
     id=#5
     possessed_class=Image
     tracks [{
       type=MovieSceneWidgetMaterialTrack
       sections [{
         type=MovieSceneParameterSection
         end_frame=3
         start_frame=-1
         channels [
          Mask Size {type=MovieSceneScriptingFloatChannel, const [0.3,2]}
          Lightning Opacity {type=MovieSceneScriptingFloatChannel, thinned_from=3, keys [[-0.0003,0],[3,0.6]]}
         ]
        }]
      }]
    }
    Reticle_Material {
     id=#6
     possessed_class=Image
     tracks [{
       type=MovieSceneWidgetMaterialTrack
       sections [{
         type=MovieSceneParameterSection
         end_frame=3
         start_frame=0
         channels [
          Outer Brackets Opacity {type=MovieSceneScriptingFloatChannel, const [1,2]}
          TopBracketOpacity {type=MovieSceneScriptingFloatChannel, keys [[0,0.5],[3,1]]}
         ]
        }]
      }]
    }
   ]}
 }
]
graphs [
 CacheReticleDynamicMaterial {nodes [
   CacheReticleDynamicMaterial() {is=title K2Node_FunctionEntry}
   > n1 = GetDynamicMaterial(self=Get Reticle_Material() -> (Reticle_Material:Image Object Reference))
   > Set ReticleDynamicMaterial(ReticleDynamicMaterial:Material Instance Dynamic Object Reference=@n1) -> (Output_Get:Material Instance Dynamic Object Reference)
  ]}
 EventGraph {nodes [
   ResetReticleBracketRotation() {is=title K2Node_CustomEvent}
   > n1 = Get Scalar Parameter Value!(
    execute
    self:Material Instance Dynamic Object Reference=Get ReticleDynamicMaterial() -> (ReticleDynamicMaterial:Material Instance Dynamic Object Reference)
    ParameterName:Name=Center Brackets Rotation
   ) -> (then, ReturnValue:float) {is=title K2Node_CallFunction}
   > Branch(Condition="Nearly Equal (Float)"(A=@n1, B=0.34, ErrorTolerance=0.1))
   > PlayAnimation(InAnimation=Get anim_2HitCombo() -> (anim_2HitCombo:Widget Animation Object Reference), PlayMode=Reverse)
   n2 = HandlePrimaryAbility!() -> (OutputDelegate, then, TechniqueDataIndex:int, TechniqueData:"Fort Mantis Technique Data Structure (by ref)", bIsLockedOn:bool=~) {is=title K2Node_CustomEvent}
   n3 = Get anim_2HitCombo()
   n4 = Get anim_3HitCombo() -> (anim_3HitCombo:Widget Animation Object Reference)
   n5 = Get anim_1HitCombo() -> (anim_1HitCombo:Widget Animation Object Reference)
   n6 = Get MaxComboHits() -> (MaxComboHits:int=0)
   > PlayAnimation(InAnimation=Select(
     Option 0:Widget Animation Object Reference=Select!(Option 0=@n5, Option 1=@n3, Option 2:Widget Animation Object Reference=@n4, Index:int=@n2.TechniqueDataIndex) -> (ReturnValue)
     Option 1:Widget Animation Object Reference=Get anim_1HitCombo()
     Index="integer >= integer"(A=@n2.TechniqueDataIndex, B=@n6)
    ) -> (ReturnValue:Widget Animation Object Reference))
   > n7 = Set Timer by Event(Delegate=Create Event(), Time=1.1, bLooping=false)
   > Set ComboEventTimerHandle(ComboEventTimerHandle:Timer Handle Structure=@n7) -> (Output_Get:Timer Handle Structure)
   n8 = OnViewTargetChanged!() -> (
    OutputDelegate
    then
    PlayerController:Fort Player Controller Object Reference
    OldViewTarget:Actor Object Reference
    NewViewTarget:Actor Object Reference
   ) {is=title K2Node_CustomEvent}
   > Set ViewTargetPawn(ViewTargetPawn:Actor Object Reference=@n8.NewViewTarget) -> (Output_Get:Actor Object Reference)
   > ListenToPrimaryAbility()
   > ListenToSecondaryAbility()
   > PlayAnimation(InAnimation=Get anim_Idle() -> (anim_Idle:Widget Animation Object Reference))
   n9 = HandleSecondaryCooldown!() -> (OutputDelegate, then, StartUpdate:bool=~) {is=title K2Node_CustomEvent}
   > n10 = Branch(Condition=@n9.StartUpdate)
   > n11 = Set Timer by Event(Delegate=Create Event(), Time=0.1, bLooping=true, bMaxOncePerFrame=true)
   > Set SecondaryCooldownTimerHandle(SecondaryCooldownTimerHandle:Timer Handle Structure=@n11) -> (Output_Get:Timer Handle Structure)
   Clear and Invalidate Timer by Handle(execute=@n10.else, Handle=Get SecondaryCooldownTimerHandle() -> (SecondaryCooldownTimerHandle:Timer Handle Structure))
   > SetScalarParameterValue(self=Get ReticleDynamicMaterial(), ParameterName=Get CooldownParameterName() -> (CooldownParameterName:Name=None), Value=1)
   ListenToPrimaryAbility() {is=title K2Node_CustomEvent}
   > Bind Event to On Technique Started(
    self:Fort Mantis Pawn Component Object Reference=GetComponentByClass!(
     self:Actor Object Reference=Get ViewTargetPawn() -> (ViewTargetPawn:Actor Object Reference)
     ComponentClass:Actor Component Class Reference=/Script/MantisRuntime.FortMantisPawnComponent
    ) -> (ReturnValue:Fort Mantis Pawn Component Object Reference)
    Delegate=Create Event()
   ) {is=title K2Node_AddDelegate}
   ListenToSecondaryAbility() {is=title K2Node_CustomEvent}
   > n12 = Sequence()
   n13 = Get ViewTargetPawn()
   > n14 = WaitGameplayTagAddToActor!(
    execute:Exec
    TargetActor:Actor Object Reference=@n13
    Tag:Gameplay Tag Structure=Get SecondaryAbilityCooldownTag() -> (SecondaryAbilityCooldownTag:Gameplay Tag Structure)
    OnlyTriggerOnce:bool=false
   ) -> (then:Exec, AsyncTaskProxy:Ability Async Wait Gameplay Tag Added Object Reference, Added:Exec) {is=title K2Node_AsyncAction}
   HandleSecondaryAbility!(execute=@n14.Added, self, IsActive:bool=true) -> (then)
   n15 = WaitGameplayTagRemoveFromActor!(
    execute:Exec=@n12.then_1
    TargetActor:Actor Object Reference=@n13
    Tag:Gameplay Tag Structure=Get SecondaryAbilityCooldownTag()
    OnlyTriggerOnce:bool=false
   ) -> (then:Exec, AsyncTaskProxy:Ability Async Wait Gameplay Tag Removed Object Reference, Removed:Exec) {is=title K2Node_AsyncAction}
   HandleSecondaryAbility(execute=@n15.Removed, IsActive=false)
   n16 = HandleSecondaryAbility!() -> (OutputDelegate, then, IsActive:bool=~) {is=title K2Node_CustomEvent}
   > PlayAnimation(InAnimation=Select!(
     Option 0=Get anim_SecondaryAttackCooldownFinished() -> (anim_SecondaryAttackCooldownFinished:Widget Animation Object Reference)
     Option 1=Get anim_SecondaryAttack() -> (anim_SecondaryAttack:Widget Animation Object Reference)
     Index:bool=@n16.IsActive
    ) -> (ReturnValue))
   > HandleSecondaryCooldown!(execute, self, StartUpdate:bool=@n16.IsActive) -> (then) {is=title K2Node_CallFunction}
   Event OnInitialized()
   > CacheReticleDynamicMaterial() {is=title K2Node_CallFunction}
   Event Construct()
   > n17 = Cast To FortPlayerController(Object=GetOwningPlayer())
   > Bind Event to On View Target Changed(self:Fort Player Controller Object Reference=@n17, Delegate=Create Event()) {is=title K2Node_AddDelegate}
   > OnViewTargetChanged!(
    execute
    self
    PlayerController:Fort Player Controller Object Reference
    OldViewTarget:Actor Object Reference
    NewViewTarget:Actor Object Reference=GetViewTarget!(self:Controller Object Reference=@n17) -> (ReturnValue:Actor Object Reference)
   ) -> (then) {is=title K2Node_CallFunction}
   Event Destruct()
   > Unbind Event from On Technique Started(
    self:Fort Mantis Pawn Component Object Reference=GetComponentByClass(self=Get ViewTargetPawn(), ComponentClass=&FortMantisPawnComponent)
    Delegate=Create Event()
   ) {is=title K2Node_RemoveDelegate}
   > n18 = Cast To FortPlayerController(Object=GetOwningPlayer())
   > Unbind Event from On View Target Changed(self:Fort Player Controller Object Reference=@n18, Delegate=Create Event()) {is=title K2Node_RemoveDelegate}
   > Clear and Invalidate Timer by Handle(Handle=Get SecondaryCooldownTimerHandle())
  ]}
 UpdateSecondaryCooldown {nodes [
   UpdateSecondaryCooldown() {is=title K2Node_FunctionEntry}
   > n1 = GetCooldownTimeRemainingAndDuration!(
    execute
    Actor:Actor Object Reference=Get ViewTargetPawn()
    CooldownTags:Gameplay Tag Container Structure=Get SecondaryCooldownTag() -> (SecondaryCooldownTag:Gameplay Tag Container Structure)
   ) -> (then, TimeRemaining:float, CooldownDuration:float)
   n2 = SafeDivide(A=float - float(A=@n1.CooldownDuration, B=@n1), B=@n1.CooldownDuration)
   n3 = "Nearly Equal (Float)"(A=@n2, B=0, ErrorTolerance=0)
   > SetScalarParameterValue(
    self=Get ReticleDynamicMaterial()
    ParameterName=Get CooldownParameterName() -> (CooldownParameterName=None)
    Value=SelectFloat(A=1, B=@n2, bPickA=@n3)
   )
   > Branch(Condition=@n3)
   > Clear and Invalidate Timer by Handle(Handle=Get SecondaryCooldownTimerHandle())
  ]}
]
variables [
 ComboEventTimerHandle {type=TimerHandle}
 CooldownParameterName {type=string, default=CooldownProgress}
 MaxComboHits {type=integer, default=3}
 ReticleDynamicMaterial {type=MaterialInstanceDynamic}
 SecondaryAbilityCooldownTag {type=GameplayTag, default {tagName=Granted.Cooldown.DentureSay.Secondary}}
 SecondaryCooldownTag {type=GameplayTagContainer, default {gameplayTags [{tagName=Granted.Cooldown.DentureSay.Secondary}]}}
 SecondaryCooldownTimerHandle {type=TimerHandle}
 ViewTargetPawn {type=Actor, default=null}
]
widgets {
 info {
  inheritedWidgetCount=1
  parentClass=/Script/FortniteUI.FortChargedWeaponReticleExtensionWidget
  rootWidgetClass=&Overlay
  widgetCount=4
 }
 widgets [
  Reticle_Overlay {widgetClassPath=&Overlay, widget=Reticle_Overlay}
  Reticle_Material {
   widgetClassPath=&Image
   bIsVariable=true
   parent=Reticle_Overlay
   widget=Reticle_Material
   properties {brush {
     bIsDynamicallyLoaded=false
     drawAs=Image
     imageType=FullColor
     mirroring=NoMirror
     resourceName=None
     resourceObject=/DentureSay/UI/Material/MI_UI_DentureSay_Reticle
     tiling=NoTile
     imageSize=<110 110>
     margin=<ltrb 0 0 0 0>
     outlineSettings {
      bUseBrushTransparency=false
      roundingType=HalfHeightRadius
      width=0
      color {colorUseRule=UseColor_Specified, specifiedColor=<rgba 0 0 0 0>}
      cornerRadii=<0 0 0 0>
     }
     tintColor {colorUseRule=UseColor_Specified, specifiedColor=<rgba 1 1 1 1>}
     uVRegion {bIsValid=false, max=<0 0>, min=<0 0>}
    }}
   slot {class=OverlaySlot, properties {horizontalAlignment=HAlign_Center, verticalAlignment=VAlign_Center}}
  }
  Image_Lightning {
   widgetClassPath=&Image
   bIsVariable=true
   parent=Reticle_Overlay
   widget=Image_Lightning
   properties {brush {
     bIsDynamicallyLoaded=false
     drawAs=Image
     imageType=FullColor
     mirroring=NoMirror
     resourceName=None
     resourceObject=/DentureSay/UI/Material/MI_UI_DenturySay_Lightning
     tiling=NoTile
     imageSize=<100 40>
     margin=<ltrb 0 0 0 0>
     outlineSettings {
      bUseBrushTransparency=false
      roundingType=HalfHeightRadius
      width=0
      color {colorUseRule=UseColor_Specified, specifiedColor=<rgba 0 0 0 0>}
      cornerRadii=<0 0 0 0>
     }
     tintColor {colorUseRule=UseColor_Specified, specifiedColor=<rgba 1 1 1 1>}
     uVRegion {bIsValid=false, max=<0 0>, min=<0 0>}
    }}
   slot {class=OverlaySlot, properties {horizontalAlignment=HAlign_Center, verticalAlignment=VAlign_Center, padding=<ltrb 0 -45 0 0>}}
  }
  TutorialHighlight {widgetClassPath=&TutorialHighlight}
 ]
}
